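# Conversation database for the GumshoeDark NPC, defining its greeting and the hub of conversation branches it offers.
# The tagged entry point leads to one Speech node with a default line plus curve-prioritized alternatives for a first-ever meeting and a Tactical Santa outfit reaction.
# That node fans out through link nodes to other entry tags: NPC library quests, basic services, the character's own dialogue, and multi-part Season 33 WF quest conversations.
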ConversationDatabase /NPCLibrary/NPCs/GumshoeDark/NPCLibrary_Conversation_GumshoeDark
entry_points [{tag=Conversation.Entry.GumshoeDark, destinations [#1]}]
nodes [
 ConversationEntryPointNode_7 {
  id=#1
  class=ConversationEntryPointNode
  outputs [#2]
  properties {entryTag {tagName=Conversation.Entry.GumshoeDark}}
 }
 Speech_C_1 {
  id=#2
  class=Speech_C
  outputs [
   #3
   #4
   #5
   #6
   #7
   #8
   #9
   #10
   #11
   #12
   #13
   #14
   #15
   #16
   #17
   #18
   #19
   #20
   #21
   #22
   #23
   #24
   #25
   #26
   #27
   #28
   #29
   #30
   #31
   #32
   #33
   #34
   #35
   #36
   #37
   #38
  ]
  properties {
   nodeName=""
   generalConfig {
    defaultMessage=Nothing's black or white. It's all a shade of noir.
    contextualMessages [
     {
      message=You need a detective for something?
      contextRequirements [{requirement {class=FortControllerRequirement_IsFirstEverConversationWithNPC}}]
      priority {curve {rowName=Tandem.Speech.Priority.FirstEverInteraction}}
     }
     {
      message="Still vetting those names for your list. Lotta Nice this year. "
      contextRequirements [{requirement {
         class=FortControllerRequirement_HasCID
         properties {allowedCIds [/BRCosmetics/Athena/Items/Cosmetics/Characters/CID_279_Athena_Commando_M_TacticalSanta]}
        }}]
      priority {curve {rowName=Tandem.Speech.Priority.OutfitReaction}}
     }
    ]
   }
  }
 }
 ConversationLinkNode_1 {
  id=#38
  class=ConversationLinkNode
  properties {remoteEntryTag {tagName=Conversation.Entry.GumshoeDark.Dialogue}}
 }
 ConversationLinkNode_33 {
  id=#37
  class=ConversationLinkNode
  properties {remoteEntryTag {tagName=Conversation.Entry.Season33.WF.Quest.06.part.05}}
 }
 ConversationLinkNode_34 {
  id=#36
  class=ConversationLinkNode
  properties {remoteEntryTag {tagName=Conversation.Entry.Season33.WF.Quest.06.part.04}}
 }
 ConversationLinkNode_30 {
  id=#35
  class=ConversationLinkNode
  properties {remoteEntryTag {tagName=Conversation.Entry.Season33.WF.Quest.06.part.03}}
 }
 ConversationLinkNode_31 {
  id=#34
  class=ConversationLinkNode
  properties {remoteEntryTag {tagName=Conversation.Entry.Season33.WF.Quest.06.part.02}}
 }
 ConversationLinkNode_32 {
  id=#33
  class=ConversationLinkNode
  properties {remoteEntryTag {tagName=Conversation.Entry.Season33.WF.Quest.06.part.01}}
 }
 ConversationLinkNode_28 {
  id=#32
  class=ConversationLinkNode
  properties {remoteEntryTag {tagName=Conversation.Entry.Season33.WF.Quest.05.part.05}}
 }
 ConversationLinkNode_29 {
  id=#31
  class=ConversationLinkNode
  properties {remoteEntryTag {tagName=Conversation.Entry.Season33.WF.Quest.05.part.04}}
 }
 ConversationLinkNode_25 {
  id=#30
  class=ConversationLinkNode
  properties {remoteEntryTag {tagName=Conversation.Entry.Season33.WF.Quest.05.part.03}}
 }
 ConversationLinkNode_26 {
  id=#29
  class=ConversationLinkNode
  properties {remoteEntryTag {tagName=Conversation.Entry.Season33.WF.Quest.05.part.02}}
 }
 ConversationLinkNode_27 {
  id=#28
  class=ConversationLinkNode
  properties {remoteEntryTag {tagName=Conversation.Entry.Season33.WF.Quest.05.part.01}}
 }
 ConversationLinkNode_23 {
  id=#27
  class=ConversationLinkNode
  properties {remoteEntryTag {tagName=Conversation.Entry.Season33.WF.Quest.04.part.05}}
 }
 ConversationLinkNode_24 {
  id=#26
  class=ConversationLinkNode
  properties {remoteEntryTag {tagName=Conversation.Entry.Season33.WF.Quest.04.part.04}}
 }
 ConversationLinkNode_20 {
  id=#25
  class=ConversationLinkNode
  properties {remoteEntryTag {tagName=Conversation.Entry.Season33.WF.Quest.04.part.03}}
 }
 ConversationLinkNode_21 {
  id=#24
  class=ConversationLinkNode
  properties {remoteEntryTag {tagName=Conversation.Entry.Season33.WF.Quest.04.part.02}}
 }
 ConversationLinkNode_22 {
  id=#23
  class=ConversationLinkNode
  properties {remoteEntryTag {tagName=Conversation.Entry.Season33.WF.Quest.04.part.01}}
 }
 ConversationLinkNode_18 {
  id=#22
  class=ConversationLinkNode
  properties {remoteEntryTag {tagName=Conversation.Entry.Season33.WF.Quest.03.part.05}}
 }
 ConversationLinkNode_19 {
  id=#21
  class=ConversationLinkNode
  properties {remoteEntryTag {tagName=Conversation.Entry.Season33.WF.Quest.03.part.04}}
 }
 ConversationLinkNode_15 {
  id=#20
  class=ConversationLinkNode
  properties {remoteEntryTag {tagName=Conversation.Entry.Season33.WF.Quest.03.part.03}}
 }
 ConversationLinkNode_16 {
  id=#19
  class=ConversationLinkNode
  properties {remoteEntryTag {tagName=Conversation.Entry.Season33.WF.Quest.03.part.02}}
 }
 ConversationLinkNode_17 {
  id=#18
  class=ConversationLinkNode
  properties {remoteEntryTag {tagName=Conversation.Entry.Season33.WF.Quest.03.part.01}}
 }
 ConversationLinkNode_13 {
  id=#17
  class=ConversationLinkNode
  properties {remoteEntryTag {tagName=Conversation.Entry.Season33.WF.Quest.02.part.05}}
 }
 ConversationLinkNode_14 {
  id=#16
  class=ConversationLinkNode
  properties {remoteEntryTag {tagName=Conversation.Entry.Season33.WF.Quest.02.part.04}}
 }
 ConversationLinkNode_10 {
  id=#15
  class=ConversationLinkNode
  properties {remoteEntryTag {tagName=Conversation.Entry.Season33.WF.Quest.02.part.03}}
 }
 ConversationLinkNode_11 {
  id=#14
  class=ConversationLinkNode
  properties {remoteEntryTag {tagName=Conversation.Entry.Season33.WF.Quest.02.part.02}}
 }
 ConversationLinkNode_12 {
  id=#13
  class=ConversationLinkNode
  properties {remoteEntryTag {tagName=Conversation.Entry.Season33.WF.Quest.02.part.01}}
 }
 ConversationLinkNode_35 {
  id=#12
  class=ConversationLinkNode
  properties {remoteEntryTag {tagName=Conversation.Entry.Season33.WF.Quest.01.part.01}}
 }
 ConversationLinkNode_6 {
  id=#11
  class=ConversationLinkNode
  properties {remoteEntryTag {tagName=Conversation.Entry.Season33.WF.Quest.01.part.02}}
 }
 ConversationLinkNode_36 {
  id=#10
  class=ConversationLinkNode
  properties {remoteEntryTag {tagName=Conversation.Entry.Season33.WF.Quest.01.part.03}}
 }
 ConversationLinkNode_9 {
  id=#9
  class=ConversationLinkNode
  properties {remoteEntryTag {tagName=Conversation.Entry.Season33.WF.Quest.01.part.04}}
 }
 ConversationLinkNode_8 {
  id=#8
  class=ConversationLinkNode
  properties {remoteEntryTag {tagName=Conversation.Entry.Season33.WF.Quest.01.part.05}}
 }
 ConversationLinkNode_3 {
  id=#7
  class=ConversationLinkNode
  properties {remoteEntryTag {tagName=Conversation.Entry.Season33.WF.Quest.00.part.02}}
 }
 ConversationLinkNode_2 {
  id=#6
  class=ConversationLinkNode
  properties {remoteEntryTag {tagName=Conversation.Entry.Season33.WF.Quest.00.part.01}}
 }
 ConversationLinkNode_4 {
  id=#5
  class=ConversationLinkNode
  properties {remoteEntryTag {tagName=Conversation.Entry.Season33.WF.Quest.00.part.03}}
 }
 ConversationLinkNode_0 {id=#4, class=ConversationLinkNode, properties {remoteEntryTag {tagName=Conversation.Entry.BasicServices}}}
 ConversationLinkNode_5 {id=#3, class=ConversationLinkNode, properties {remoteEntryTag {tagName=Conversation.Entry.NPCLibrary.Quests}}}
]
