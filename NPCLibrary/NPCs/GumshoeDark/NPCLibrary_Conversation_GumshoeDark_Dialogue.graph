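# Conversation database for the GumshoeDark NPC's dialogue, offering an "Ask about..." topic on vetting Sgt. Winter's Nice and Naughty List.
# The entry point leads to an opening Speech node whose outputs are the list topic or a Back node, each shown as a player choice.
# The topic plays as a linear chain of four Speech lines ending in a ReturnToStart node labeled BACK.
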ConversationDatabase /NPCLibrary/NPCs/GumshoeDark/NPCLibrary_Conversation_GumshoeDark_Dialogue
entry_points [{tag=Conversation.Entry.GumshoeDark.Dialogue, destinations [#1]}]
nodes [
 ConversationEntryPointNode_0 {
  id=#1
  class=ConversationEntryPointNode
  outputs [#2]
  properties {entryTag {tagName=Conversation.Entry.GumshoeDark.Dialogue}}
 }
 Speech_C_0 {
  id=#2
  class=Speech_C
  outputs [#3, #4]
  properties {nodeName="", softTaskIcon=&T_UI_NPC_question, generalConfig {defaultMessage=I'll clue you in.}}
  sub_nodes [ConversationChoiceNode_5 {class=ConversationChoiceNode, properties {defaultChoiceDisplayText=Ask about...}}]
 }
 Back_C_0 {
  id=#4
  class=Back_C
  sub_nodes [ConversationChoiceNode_7 {class=ConversationChoiceNode, properties {defaultChoiceDisplayText=BACK}}]
 }
 Speech_C_7 {
  id=#3
  class=Speech_C
  outputs [#5]
  properties {
   nodeName=""
   softTaskIcon=&T_UI_NPC_question
   generalConfig {defaultMessage="Every year, I help vet Sgt. Winter's Nice and Naughty List."}
  }
  sub_nodes [ConversationChoiceNode_15 {class=ConversationChoiceNode, properties {defaultChoiceDisplayText="The Nice && Naughty List"}}]
 }
 Speech_C_15 {
  id=#5
  class=Speech_C
  outputs [#6]
  properties {
   nodeName=""
   softTaskIcon=&T_UI_NPC_question
   generalConfig {defaultMessage=Check who's nice. Check who's naughty. The works.}
  }
 }
 Speech_C_14 {
  id=#6
  class=Speech_C
  outputs [#7]
  properties {
   nodeName=""
   softTaskIcon=&T_UI_NPC_question
   generalConfig {defaultMessage=Sometimes I think he wishes he could ditch the Nice list and throw everyone in Naughty.}
  }
 }
 Speech_C_1 {
  id=#7
  class=Speech_C
  outputs [#8]
  properties {
   nodeName=""
   softTaskIcon=&T_UI_NPC_question
   generalConfig {defaultMessage="It's a lot of footwork, but hey, I'm a detective."}
  }
 }
 ReturnToStart_C_23 {
  id=#8
  class=ReturnToStart_C
  properties {softTaskIcon=&T_UI_NPC_Back}
  sub_nodes [ConversationChoiceNode_32 {class=ConversationChoiceNode, properties {defaultChoiceDisplayText=BACK}}]
 }
]
